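; StarBeam vehicle boost cosmetic Blueprint that applies paint color to its Niagara boost effect.
; SetupPaintSettings calls the parent to resolve primary color, balanced and metallic values, then returns them unchanged.
; It pushes the primary color into both shape color parameters on NS_Boost and sets ColorWhite to a constant, independent of the paint.
; BeginPlay, overlap and Tick are empty, and the construction script only calls its parent.
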(event EventBeginPlay)

(event Collision|EventActorBeginOverlap
  (params (OtherActor :type Actor%20Object%20Reference)))

(event EventTick
  (params (DeltaSeconds :type Float%20%28single-precision%29)))

(fn SetupPaintSettings
  (params (Color :type Name :default "None") (AdditionalInfo :type Vehicle%20Cosmetics%20Additional%20Parameter%20Infos%20Structure))
  (bind _ns_boost (Variables|Default|GetNS_Boost))
  (bind (_primarycolor _balanced _metallic) (|CallParentFunction Color AdditionalInfo))
  (Niagara|SetNiagaraVariable(LinearColor) _ns_boost "Shape01" _primarycolor)
  (Niagara|SetNiagaraVariable(LinearColor) _ns_boost "Shape02" _primarycolor)
  (Niagara|SetNiagaraVariable(Float) _ns_boost "ColorWhite" 1.0)
  (return _primarycolor _balanced _metallic))

(fn ConstructionScript
  (|CallParentFunction))
