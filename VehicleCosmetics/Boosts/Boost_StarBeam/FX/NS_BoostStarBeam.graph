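# Niagara system for the Star Beam vehicle boost cosmetic, attached to skeletal mesh sockets with a per-car offset.
# BlueStars002 merges four star sprite variants via spawn groups, partitioned across boost ports, colored blue/yellow or by painted shape colors.
# RibbonTrail001 spawns per-port ribbons with five renderers, using painted-variant overrides and curve-driven color and width layers.
# A frontend-preview flag adds backward velocity and changes some spawn rates.
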NiagaraSystem /VehicleCosmetics/Boosts/Boost_StarBeam/FX/NS_BoostStarBeam
dependencies {
 usedDataInterfaces [
  &NiagaraDataInterfaceArrayFloat
  &NiagaraDataInterfaceSpriteRendererInfo
  &NiagaraDataInterfaceArrayInt32
  &NiagaraDataInterfaceCurve
  &NiagaraDataInterfaceSkeletalMesh
 ]
 usedDynamicInputs []
 usedModules [
  &self:SystemSpawnScript.SystemScriptSource.SystemScriptGraph.NiagaraNodeAssignment_0.SetVariables_#1
  &SystemState
 ]
}
emitters {
 BlueStars002 {
  emitterUpdateScript [
   &EmitterState()
   &VC_FilteredSocketCount()
   &SpawnRate(
    SpawnRate=&MakeCustomFloatFromBool(Bool=$User.bIsFrontendPreview, True Float=12.5, False Float=20)
    Use Spawn Probability=false
   )
   &SpawnRate(
    SpawnGroup=1
    SpawnRate=&MakeCustomFloatFromBool(Bool=$User.bIsFrontendPreview, True Float=25, False Float=45)
    Use Spawn Probability=false
   ) {module=SpawnRate001}
   &SpawnRate(
    SpawnGroup=2
    SpawnRate=&MakeCustomFloatFromBool(Bool=$User.bIsFrontendPreview, True Float=40, False Float=40)
    Use Spawn Probability=false
   ) {module=SpawnRate002}
   &SpawnRate(
    SpawnGroup=3
    SpawnRate=&MakeCustomFloatFromBool(Bool=$User.bIsFrontendPreview, True Float=40, False Float=40)
    Use Spawn Probability=false
   ) {module=SpawnRate003}
  ]
  particleSpawnScript [
   &VC_PaintedVariantsOverride() {enabled=false}
   &InitializeParticle(
    Color=<rgba 1 0 0.0667 1>
    Lifetime Max=&SelectFloatFromArray(
     Direct Array Index=&SpawnGroup()
     Float Selection Array=DataInterface() {propertyValues {FloatData [0.35,0.31,0.35,0.31]}}
    )
    Lifetime Min:NiagaraFloat=&SelectFloatFromArray(
     Direct Array Index=&SpawnGroup()
     Float Selection Array=DataInterface() {propertyValues {FloatData [0.25,0.23,0.251,0.23]}}
    )
    Lifetime Mode=::NewEnumerator1
    Sprite Rotation Mode=::NewEnumerator2
    Sprite Size Mode=::NewEnumerator1
    Uniform Sprite Size Max=&SelectFloatFromArray(Direct Array Index=$Emitter.SpawnGroup, Float Selection Array=DataInterface() {propertyValues {FloatData [15,25,15,15]}})
    Uniform Sprite Size Min=&SelectFloatFromArray(Direct Array Index=$Emitter.SpawnGroup, Float Selection Array=DataInterface() {propertyValues {FloatData [10,18,8,9]}})
    UsePositionOffset=false
    Write Material Random=true
   )
   &PartitionParticles(Number Of Partitions=$User.PortCount)
   &SkeletalMeshLocation(
    Bone / Socket Index=$Particles.PartitionParticles.Partition
    Bone / Socket Sampling Mode=::NewEnumerator5
    MaskSpawnGroup=false
    Mesh Sampling Type=::NewEnumerator6
    Sampled Position Offset=&AddVectorToPosition(Position=<-25 0 0>, Vector=$User.DTCarOffset)
    Skeletal Mesh=$System.SkeletalMesh
   )
   &AddVelocity(
    Cone Angle=&SelectFloatFromArray(
     Direct Array Index=&SpawnGroup()
     Float Selection Array=DataInterface() {propertyValues {FloatData [35,35,45,80]}}
    )
    Cone Axis=<-1 0 0>
    Module.UseConeFalloff=false
    Module.UseVelDistribution=false
    Override Seed=false
    Velocity Mode=::NewEnumerator2
    Velocity Speed=&RandomRangeFloat(
     Minimum=/VehicleCosmetics/FX/Niagara/Modules/MergeEmitters/VC_CompareBySpawnGroup(Masked Value:NiagaraFloat=500, Mask Value:NiagaraInt32=1, Passtrough Value:NiagaraFloat=15, Undefined parameter name:NiagaraInt32=1)
     Maximum=&VC_CompareBySpawnGroup(Masked Value:NiagaraFloat=1000, Mask Value:NiagaraInt32=1, Passtrough Value:NiagaraFloat=20, Undefined parameter name:NiagaraInt32=1)
    )
   ) {module=AddVelocity001}
   &ShapeLocation(
    MaskSpawnGroup=false
    Override Seed=false
    Sphere Radius=&RandomRangeFloat(
     Minimum=&SelectFloatFromArray(
      Direct Array Index=&SpawnGroup()
      Float Selection Array=DataInterface() {propertyValues {FloatData [5,20,20,10]}}
     )
     Maximum=20
    )
    UseOwnerScale=true
   )
   &SubUVAnimation(
    End Frame Range Override:NiagaraInt32=&SelectIntFromArray(Direct Array Index=&SpawnGroup(), Int Selection Array=DataInterface() {propertyValues {IntData [3,2,3,4]}})
    "Sprite Renderer "=DataInterface() {propertyValues {SpriteRenderer=&self:BlueStars002_1.NiagaraSpriteRendererProperties_2}}
    Start Frame Range Override=&SelectIntFromArray(Direct Array Index=$Emitter.SpawnGroup, Int Selection Array=DataInterface() {propertyValues {IntData [0,1,0,2]}})
    SubUV Animation Mode:ENiagara_SubUVLookupModeV2=::NewEnumerator2
    UseEndFrame=true
    UseStartFrame=true
   )
   &AddVelocity(Velocity=&MakeCustomVectorFromBool(
     Bool=$User.bIsFrontendPreview
     True Vector=&MakeVector(
      X=&VC_CompareBySpawnGroup(Masked Value:NiagaraFloat=0, Mask Value:NiagaraInt32=1, Passtrough Value:NiagaraFloat=-1500, Undefined parameter name:NiagaraInt32=1)
      Y=0
      Z=0
     )
     False Vector=<0 0 0>
    ))
   /SetVariables(
    Particles.BlueColor:LinearColor=&MakeCustomLinearColorFromBool(Bool=$User.bIsPainted, True Color=$User.Shape01, False Color=<rgba 0 0.7784 1 1>)
    Particles.YellowColor:LinearColor=&MakeCustomLinearColorFromBool(Bool=$User.bIsPainted, True Color=$User.Shape02, False Color=<rgba 1 0.9 0 1>)
   ) {isSetParameters=true}
   &Color(Color=/VehicleCosmetics/FX/Niagara/Modules/MergeEmitters/VC_Dynamic_ColorBySpawnGroup(
     Color SG 0:LinearColor=&MakeCustomLinearColorFromBool(Bool=$User.bIsPainted, True Color=$User.Shape01, False Color=<rgba 0 0.7784 1 1>)
     MaximumSpawnGroups:NiagaraInt32=4
     Color SG 1:LinearColor=$Particles.YellowColor
     Color SG 2:LinearColor=$Particles.BlueColor
     Color SG 3:LinearColor=$Particles.YellowColor
     Color SG 4:LinearColor=<rgba 1 1 1 1>
     SpawnGroup:NiagaraInt32=$Emitter.SpawnGroup
    ))
  ]
  particleUpdateScript [
   &ParticleState()
   &ScaleSpriteSize(Uniform Curve Sprite Scale=DataInterface() {propertyValues {Curve {keys [{interpMode=RCIM_Cubic, tangentMode=RCTM_User, time=0, value=1}, {interpMode=RCIM_Cubic, tangentMode=RCTM_User, time=1, value=0}]}}})
   &ScaleColor(Scale RGB=&VectorFromFloat(Value=$User.Emissive), ScaleA=true, ScaleRGB=true)
   &SolveForcesAndVelocity(Clamp Velocity=false, Limit Acceleration=false)
  ]
  rendererClasses [&NiagaraSpriteRendererProperties]
  renderers [{rendererClass=&NiagaraSpriteRendererProperties}]
 }
 RibbonTrail001 {
  emitterUpdateScript [
   &EmitterState()
   &VC_RateQualityHandler(High Spawn Rate=30, Low Spawn Rate=10, Medium Spawn Rate=20, Use Custom Rate?=true) {enabled=false, module=VC_RateQualityHandler001}
   &VC_FilteredSocketCount()
   &VC_RateQualityHandler()
   &VC_SpawnRibbon_v1(Number of Ribbons=$Emitter.PortCount, Spawn Time=-1, SpawnRate=$Emitter.Spawn Rate)
  ]
  particleSpawnScript [
   &VC_PaintedVariantsOverride(
    Black=<rgba 0.07 0.07 0.07 3>
    BurntSienna=<rgba 0.15 0.0463 0.0185 0.9>
    Cobalt=<rgba 0.45 0.5417 1 0.1>
    Crimson=<rgba 1 0 0.0271 0.5>
    DefaultAlpha=0.6
    DefaultColor=<rgba 0 0.5417 0.65 1>
    ForestGreen=<rgba 0 0.5 0.0253 0.6>
    Grey=<rgba 0.2031 0.2031 0.2031 0.8>
    Lime=<rgba 0.375 0.5 0 0.8>
    Orange=<rgba 0.672 0.3657 0.2016 1>
    Pink=<rgba 1 0.15 0.575 0.8>
    Purple=<rgba 0.5133 0.14 0.7 1>
    R1 Black:LinearColor=<rgba 0.2 0.2 0.2 50>
    R1 BurntSienna:LinearColor=<rgba 0.2 0.1212 0.1 40>
    R1 Cobalt:LinearColor=<rgba 0 0.0925 5 10>
    R1 Crimson:LinearColor=<rgba 2 0 0.0542 11>
    R1 DefaultColor=<rgba 1 1 0 20>
    R1 ForestGreen:LinearColor=<rgba 0 1.5 0 20>
    R1 Grey:LinearColor=<rgba 0.4 0.4 0.4 35>
    R1 Lime:LinearColor=<rgba 0.6667 1 0 40>
    R1 Orange:LinearColor=<rgba 1 0.75 0 40>
    R1 Pink:LinearColor=<rgba 1 0.45 0.8 80>
    R1 Purple:LinearColor=<rgba 1 0.5 0.8182 40>
    R1 Saffron:LinearColor=<rgba 2 1.5333 0 30>
    R1 SkyBlue:LinearColor=<rgba 0.4 1.4646 2 15>
    R1 TitaniumWhite:LinearColor=<rgba 2 2 2 15>
    R2 Black:LinearColor=<rgba 0.098 0.098 0.098 1>
    R2 BurntSienna:LinearColor=<rgba 0.12 0.0774 0.066 1>
    R2 Cobalt:LinearColor=<rgba 0 0.0925 5 0.2>
    R2 Crimson:LinearColor=<rgba 1 0.15 0.173 0.5>
    R2 DefaultColor:LinearColor=<rgba 1 1 1 0.3>
    R2 ForestGreen:LinearColor=<rgba 0.15 1 0.1929 0.35>
    R2 Grey:LinearColor=<rgba 0.3 0.3 0.3 1>
    R2 Lime:LinearColor=<rgba 0.6667 1 0 0.6>
    R2 Orange:LinearColor=<rgba 1 0.5833 0 1>
    R2 Pink:LinearColor=<rgba 1 0.5 0.8182 0.5>
    R2 Purple:LinearColor=<rgba 1 0.3 0.7455 1>
    R2 Saffron:LinearColor=<rgba 1.5 1.15 0 1>
    R2 SkyBlue:LinearColor=<rgba 0.12 0.8386 1.2 1>
    R3 Black:LinearColor=<rgba 0.07 0.07 0.07 7>
    R3 BurntSienna:LinearColor=<rgba 0.15 0.0463 0.0185 1>
    R3 Cobalt:LinearColor=<rgba 0.16 0.1718 0.8 0.5>
    R3 DefaultAlpha=0.6
    R3 DefaultColor:LinearColor=<rgba 0 0.5417 0.65 0.2>
    R3 ForestGreen:LinearColor=<rgba 0 0.5 0.0253 0.35>
    R3 Grey:LinearColor=<rgba 0.2031 0.2031 0.2031 0.8>
    R3 Lime:LinearColor=<rgba 0.2625 0.35 0 0.8>
    R3 Orange:LinearColor=<rgba 0.672 0.3657 0.2016 0.8>
    R3 Saffron:LinearColor=<rgba 1.2 0.962 0.18 1>
    R3 TitaniumWhite:LinearColor=<rgba 0.5 0.5 0.5 1>
    Saffron=<rgba 1.2 0.948 0.12 1>
    SkyBlue=<rgba 0 0.5988 0.9 1>
    TitaniumWhite=<rgba 0.5 0.5 0.5 1>
    Use R1 Painted Overrides=true
    Use R2 Painted Overrides=true
    Use R3 Painted Overrides=true
   )
   &VC_InitializeRibbon_v1(
    Color=$Particles.Color
    Custom Offset=true
    Enable Skeletal Position=true
    Lifetime=1
    Local Vector Offset=&AddVectorToPosition(Position=<2 0 0>, Vector=$User.DTCarOffset)
    Offset Amount=1
    Ribbon Width=40
   )
   &AddVelocity(
    Velocity=&MakeCustomVectorFromBool(Bool=$User.bIsFrontendPreview, True Vector=<-2000 0 1>, False Vector=<0 0 0>)
    Velocity Speed Scale=&MakeFloatFromVector(
     Vector=&VectorFromFloat(Value=&FloatFromCurve(FloatCurve=DataInterface() {propertyValues {Curve {keys [{interpMode=RCIM_Cubic, tangentMode=RCTM_User, time=0, value=1}, {interpMode=RCIM_Cubic, tangentMode=RCTM_User, time=1, value=0}]}}}))
     Channel=::NewEnumerator0
    )
   ) {module=AddVelocity001}
  ]
  particleUpdateScript [
   &ParticleState()
   &InheritSourceMovement(
    Applied Position Scale=&VectorFromFloat(Value=&FloatFromCurve(FloatCurve=DataInterface() {propertyValues {Curve {keys [<at 0 1>,<at 1 0>]}}}))
    ApplyForce=false
    ApplyPosition=true
    ApplyPositionFalloff=false
   )
   &ScaleRibbonWidth(Ribbon Width Scale=&FloatFromCurve(FloatCurve=DataInterface() {propertyValues {Curve {keys [<at 0 1>,<at 1 0.3>]}}}))
   &SolveForcesAndVelocity(Clamp Velocity=false, Limit Acceleration=false)
   &VC_UpdateRibbon_v1(Enable Skeletal Position=true)
   &ScaleColor(
    Scale Alpha=&MakeCustomFloatFromBool(
     Bool=$User.bIsPainted
     True Float=&FloatFromCurve(FloatCurve=DataInterface() {propertyValues {Curve {
         defaultValue=1
         keys [
          {arriveTangent=-5, interpMode=RCIM_Cubic, time=0, value=0}
          {interpMode=RCIM_Cubic, tangentMode=RCTM_User, time=0.2, value=1}
          {arriveTangent=-1.25, interpMode=RCIM_Cubic, leaveTangent=-1.25, time=0.8, value=1}
          {interpMode=RCIM_Cubic, tangentMode=RCTM_User, time=1, value=0}
         ]
        }}})
     False Float=&FloatFromCurve(FloatCurve=DataInterface() {propertyValues {Curve {
         defaultValue=1
         keys [
          {interpMode=RCIM_Cubic, tangentMode=RCTM_User, time=0, value=0}
          {interpMode=RCIM_Cubic, tangentMode=RCTM_User, time=0.2, value=1}
          {arriveTangent=0.4729, interpMode=RCIM_Cubic, leaveTangent=0.4729, tangentMode=RCTM_User, time=0.5, value=0}
         ]
        }}})
    )
    ScaleA=true
    ScaleRGB=true
   )
   /SetVariables(Particles.ScaleOutLinecolor:LinearColor=&MultiplyLinearColor(
     Linear Color=$Particles.R1 Color
     Scale Factor=&MakeVector4FromVectorAndFloat(
      "Vector (RGB)"=<1 1 1>
      "Float (Alpha)"=&FloatFromCurve(FloatCurve=DataInterface() {propertyValues {Curve {
          defaultValue=1
          keys [
           {arriveTangent=-2.1875, interpMode=RCIM_Cubic, time=0, value=0}
           {interpMode=RCIM_Cubic, tangentMode=RCTM_User, time=0.2, value=1}
           {interpMode=RCIM_Cubic, tangentMode=RCTM_User, time=0.5, value=-0.05}
          ]
         }}})
     )
    )) {isSetParameters=true}
   /SetVariables(Particles.ScaleColorNoise:LinearColor=&MultiplyLinearColor(
     Linear Color=$Particles.R2 Color
     Scale Factor=&MakeVector4FromVectorAndFloat(
      "Vector (RGB)"=<1 1 1>
      "Float (Alpha)"=&FloatFromCurve(FloatCurve=DataInterface() {propertyValues {Curve {
          defaultValue=1
          keys [
           {arriveTangent=5, interpMode=RCIM_Cubic, time=0, value=0}
           {interpMode=RCIM_Cubic, tangentMode=RCTM_User, time=0.2, value=1}
           {interpMode=RCIM_Cubic, tangentMode=RCTM_User, time=0.5, value=0}
          ]
         }}})
     )
    )) {isSetParameters=true}
   /SetVariables(
    Particles.NoiseBluecolor:LinearColor=&MultiplyLinearColor(
     Linear Color=$Particles.R3 Color
     Scale Factor=&MakeVector4FromVectorAndFloat(
      "Vector (RGB)"=<1 1 1>
      "Float (Alpha)"=&FloatFromCurve(
       FloatCurve=DataInterface() {propertyValues {Curve {
          defaultValue=1
          keys [{interpMode=RCIM_Cubic, tangentMode=RCTM_User, time=0.3, value=1}, {interpMode=RCIM_Cubic, tangentMode=RCTM_User, time=0.55, value=0}]
         }}}
       Scale Curve=0.4
      )
     )
    )
    Particles.Size_NoiseFinal:NiagaraFloat=&Multiply_Float(
     A=25
     B=&FloatFromCurve(FloatCurve=DataInterface() {propertyValues {Curve {keys [{arriveTangent=0.4294, interpMode=RCIM_Cubic, time=0.4717, value=0.9872}, {interpMode=RCIM_Cubic, tangentMode=RCTM_User, time=1, value=0.3}]}}})
    )
   ) {isSetParameters=true}
   /SetVariables(Particles.SizeOutlines:NiagaraFloat=&Multiply_Float(
     A=40
     B=&FloatFromCurve(FloatCurve=DataInterface() {propertyValues {Curve {keys [<at 0 1>,<at 1 0.5>]}}})
    )) {isSetParameters=true}
   /SetVariables(Particles.float:NiagaraFloat=&Multiply_Float(
     A=35
     B=&FloatFromCurve(FloatCurve=DataInterface() {propertyValues {Curve {keys [<at 0 1>,<at 1 0.3>]}}})
    )) {isSetParameters=true}
  ]
  rendererClasses [&NiagaraRibbonRendererProperties]
  renderers [
   {rendererClass=&NiagaraRibbonRendererProperties}
   {rendererClass=&NiagaraRibbonRendererProperties, rendererIndex=1}
   {rendererClass=&NiagaraRibbonRendererProperties, rendererIndex=2}
   {rendererClass=&NiagaraRibbonRendererProperties, rendererIndex=3}
   {rendererClass=&NiagaraRibbonRendererProperties, rendererIndex=4}
  ]
 }
}
summary {
 systemName=NS_BoostStarBeam
 emitters [
  {emitterName=BlueStars002, rendererClasses [&NiagaraSpriteRendererProperties]}
  {emitterName=RibbonTrail001, rendererClasses [&NiagaraRibbonRendererProperties]}
 ]
 userVariables [
  User.bIsFrontendPreview {type=NiagaraBool, defaultValue=true}
  User.bIsPainted {type=NiagaraBool}
  User.ColorVariantID {type=NiagaraInt32}
  User.ColorWhite {type=NiagaraFloat, defaultValue=0, description=""}
  User.DTCarOffset {type=Vector3f, description="", defaultValue {struct=Vector3f, value=<0 0 0>}}
  User.Emissive {type=NiagaraFloat, defaultValue=10, description=""}
  User.EmissiveTrail {type=NiagaraFloat, defaultValue=1, description=""}
  User.EmissiveTrailAlpha {type=NiagaraFloat, defaultValue=0.6, description=""}
  User.PaintedColor {type=LinearColor, description="", defaultValue {struct=LinearColor, value=<rgba 1 1 1 1>}}
  User.PaintedScalar {type=NiagaraFloat, defaultValue=1, description=""}
  User.PortCount {type=NiagaraInt32}
  User.Shape01 {type=LinearColor, description="", defaultValue {struct=LinearColor, value=<rgba 0 0.1275 1 1>}}
  User.Shape02 {type=LinearColor, description="", defaultValue {struct=LinearColor, value=<rgba 1 0 0.0097 1>}}
 ]
}
